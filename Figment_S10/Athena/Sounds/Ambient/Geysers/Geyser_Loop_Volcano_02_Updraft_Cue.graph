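# Looping ambient Sound Cue, named as the updraft layer of a volcano geyser, routed to an outdoor-ducking ambient sound class.
# A Mixer blends a reused skytube wind-conduit loop with an airship blade-rotation loop, each through its own attenuation node.
# The wind layer uses a capsule falloff with listener focus, while the blade layer is non-spatialized and low-pass filtered by distance.
# Cue-level overrides add capsule attenuation with occlusion and limit concurrency, stopping the quietest instance.
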SoundCue /Figment_S10/Athena/Sounds/Ambient/Geysers/Geyser_Loop_Volcano_02_Updraft_Cue
root=Mixer_0
nodes [
 {
  id=Attenuation_0
  children [WavePlayer_0]
  properties {
   bOverrideAttenuation=true
   attenuationOverrides {
    attenuationShape=Capsule
    bEnableListenerFocus=true
    falloffDistance=3500
    nonFocusVolumeAttenuation=0.5
    occlusionInterpolationTime=1
    occlusionLowPassFilterFrequency=1000
    occlusionTraceChannel=ECC_GameTraceChannel11
    occlusionVolumeAttenuation=0.5
    reverbWetLevelMax=20
    stereoSpread=10000
    attenuationShapeExtents=<5000 4096>
   }
  }
 }
 {
  id=Attenuation_1
  children [WavePlayer_2]
  properties {
   bOverrideAttenuation=true
   attenuationOverrides {
    bAttenuateWithLPF=true
    bSpatialize=false
    falloffDistance=5000
    lPFFrequencyAtMax=2000
    lPFFrequencyAtMin=6000
    lPFRadiusMax=5000
    occlusionInterpolationTime=3
    occlusionLowPassFilterFrequency=1000
    occlusionTraceChannel=ECC_GameTraceChannel11
    attenuationShapeExtents {x=2000}
   }
  }
 }
 {id=Mixer_0, children [Attenuation_0, Attenuation_1], properties {inputVolume [1,1]}}
 {id=WavePlayer_0, properties {bLooping=true, soundWaveAssetPtr=/Game/Athena/Sounds/Ambient/Skytube/Rev1/WindConduit_Anchor_02}}
 {id=WavePlayer_2, properties {bLooping=true, soundWaveAssetPtr=/Figment_S10/Athena/Sounds/Vehicles/Airship/Airship_Blade_RotateLoop_01}}
]
properties {
 bOverrideConcurrency=true
 maxDistance=8500
 soundClassObject=/Game/Sounds/FortSoundClassesAndModes/Fort_Ambient_DuckOutdoors_Athena_PSM
 volumeMultiplier=2
 attenuationOverrides {
  attenuationShape=Capsule
  bAttenuateWithLPF=true
  bEnableOcclusion=true
  falloffDistance=5000
  lPFFrequencyAtMax=500
  lPFRadiusMax=4000
  nonSpatializedRadiusStart=3000
  occlusionInterpolationTime=1
  occlusionLowPassFilterFrequency=1000
  occlusionTraceChannel=ECC_GameTraceChannel11
  reverbWetLevelMax=20
  stereoSpread=10000
  attenuationShapeExtents=<7000 4096>
 }
 concurrencyOverrides {maxCount=3, resolutionRule=StopQuietest}
}
